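# Physics asset for the dynamic collar on the male medium Commando Blue Badass body mesh, covering front, side and back collar pieces.
# Each of the five collar pieces is a capsule constrained with swing and twist limits to its own small base sphere.
# The base spheres and the neck_02 capsule connect to spine_05 with free angular motion.
# Large neck and upper-spine capsules give the body volume beside the collar.
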PhysicsAsset /BRCosmetics/Characters/Player/Male/Medium/Bodies/M_MED_Commando_Blue_Badass/Meshes/M_MED_Blue_Badass_Collar_Dyn_Physics
bodies {
 dyn_collar_back [{shapeType=Capsule, length=7, radius=1.75, center=<0 0.729 4.5106>, rotation=<rot 90 0 0>}]
 dyn_collar_back_base [{shapeType=Sphere, radius=0.505}]
 dyn_collar_front_l [{shapeType=Capsule, length=5.25, radius=1.5, center=<0.008 -0.1899 3.1416>, rotation=<rot 85.8633 0.0003 1.4446>}]
 dyn_collar_front_l_base [{shapeType=Sphere, radius=0.505}]
 dyn_collar_front_r [{shapeType=Capsule, length=5.25, radius=1.5, center=<0.1222 0.1597 -3.1594>, rotation=<rot 85.8619 -0.0001 -0.0001>}]
 dyn_collar_front_r_base [{shapeType=Sphere, radius=0.505}]
 dyn_collar_side_l [{shapeType=Capsule, length=4, radius=1.75, center=<0.0147 0.1963 3.6475>, rotation=<rot 88.8463 0.0002 0.0002>}]
 dyn_collar_side_l_base [{shapeType=Sphere, radius=0.505}]
 dyn_collar_side_r [{shapeType=Capsule, length=4, radius=1.75, center=<-0.005 -0.1928 -3.6633>, rotation=<rot 89.3385 -179.9996 -179.9996>}]
 dyn_collar_side_r_base [{shapeType=Sphere, radius=0.505}]
 neck_02 [{shapeType=Capsule, length=1.75, radius=5.8, center=<0.8517 0.8457 -0.0114>, rotation=<rot 0 -89.9995 90.7599>}]
 spine_05 [{shapeType=Capsule, length=4.3176, radius=6.8473, center=<1.3181 0.2648 0.3204>, rotation=<rot 86.3682 104.2556 -165.7705>}]
}
constraints [
 {bone1Name=dyn_collar_front_l, bone2Name=dyn_collar_front_l_base, swing1LimitDegrees=60, swing2LimitDegrees=60, twistLimitDegrees=60}
 {bone1Name=dyn_collar_front_r, bone2Name=dyn_collar_front_r_base, swing1LimitDegrees=60, swing2LimitDegrees=60, twistLimitDegrees=60}
 {bone1Name=dyn_collar_back, bone2Name=dyn_collar_back_base, swing1LimitDegrees=60, swing2LimitDegrees=60, twistLimitDegrees=60}
 {bone1Name=dyn_collar_side_l, bone2Name=dyn_collar_side_l_base, swing1LimitDegrees=60, swing2LimitDegrees=60, twistLimitDegrees=60}
 {bone1Name=dyn_collar_side_r, bone2Name=dyn_collar_side_r_base, swing1LimitDegrees=60, swing2LimitDegrees=60, twistLimitDegrees=60}
 {
  bone1Name=neck_02
  bone2Name=spine_05
  swing1LimitDegrees=45
  swing1Motion=Free
  swing2LimitDegrees=45
  swing2Motion=Free
  twistLimitDegrees=45
  twistMotion=Free
 }
 {
  bone1Name=dyn_collar_front_l_base
  bone2Name=spine_05
  swing1LimitDegrees=45
  swing1Motion=Free
  swing2LimitDegrees=45
  swing2Motion=Free
  twistLimitDegrees=45
  twistMotion=Free
 }
 {
  bone1Name=dyn_collar_front_r_base
  bone2Name=spine_05
  swing1LimitDegrees=45
  swing1Motion=Free
  swing2LimitDegrees=45
  swing2Motion=Free
  twistLimitDegrees=45
  twistMotion=Free
 }
 {
  bone1Name=dyn_collar_back_base
  bone2Name=spine_05
  swing1LimitDegrees=45
  swing1Motion=Free
  swing2LimitDegrees=45
  swing2Motion=Free
  twistLimitDegrees=45
  twistMotion=Free
 }
 {
  bone1Name=dyn_collar_side_l_base
  bone2Name=spine_05
  swing1LimitDegrees=45
  swing1Motion=Free
  swing2LimitDegrees=45
  swing2Motion=Free
  twistLimitDegrees=45
  twistMotion=Free
 }
 {
  bone1Name=dyn_collar_side_r_base
  bone2Name=spine_05
  swing1LimitDegrees=45
  swing1Motion=Free
  swing2LimitDegrees=45
  swing2Motion=Free
  twistLimitDegrees=45
  twistMotion=Free
 }
]
