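# Physics asset for the dynamic shoe flaps on the medium male Commando Blue Badass character body.
# Capsule bodies cover the pelvis, two twist bones per calf and two small flap bones on each side.
# Each flap is constrained to its same-side calf twist bone with asymmetric swing and tight twist limits.
# The calf twist bodies link directly to the pelvis with free angular motion, skipping the rest of the leg chain.
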PhysicsAsset /BRCosmetics/Characters/Player/Male/Medium/Bodies/M_MED_Commando_Blue_Badass/Meshes/M_MED_Blue_Badass_Dyn_Physics
bodies {
 calf_twist_01_l [{shapeType=Capsule, length=1.25, radius=2.6495}]
 calf_twist_01_r [{shapeType=Capsule, length=0.75, radius=2.6709}]
 calf_twist_02_l [{shapeType=Capsule, length=1, radius=3.4819}]
 calf_twist_02_r [{shapeType=Capsule, length=0.3923, radius=4.4272}]
 dyn_shoeFlap_a_l [{shapeType=Capsule, length=0, radius=1.755, center=<0.0446 0.0473 -0.6961>}]
 dyn_shoeFlap_a_r [{shapeType=Capsule, length=0, radius=1.755, center=<0 0 0.6991>}]
 dyn_shoeFlap_b_l [{shapeType=Capsule, length=0, radius=1.755, center=<0.0446 0.0473 -0.6961>}]
 dyn_shoeFlap_b_r [{shapeType=Capsule, length=0, radius=1.755, center=<0 0 0.6991>}]
 pelvis [{shapeType=Capsule, length=20.6235, radius=8.1306}]
}
constraints [
 {
  bone1Name=calf_twist_01_l
  bone2Name=pelvis
  swing1LimitDegrees=45
  swing1Motion=Free
  swing2LimitDegrees=45
  swing2Motion=Free
  twistLimitDegrees=45
  twistMotion=Free
 }
 {
  bone1Name=calf_twist_02_l
  bone2Name=pelvis
  swing1LimitDegrees=45
  swing1Motion=Free
  swing2LimitDegrees=45
  swing2Motion=Free
  twistLimitDegrees=45
  twistMotion=Free
 }
 {
  bone1Name=calf_twist_01_r
  bone2Name=pelvis
  swing1LimitDegrees=45
  swing1Motion=Free
  swing2LimitDegrees=45
  swing2Motion=Free
  twistLimitDegrees=45
  twistMotion=Free
 }
 {
  bone1Name=calf_twist_02_r
  bone2Name=pelvis
  swing1LimitDegrees=45
  swing1Motion=Free
  swing2LimitDegrees=45
  swing2Motion=Free
  twistLimitDegrees=45
  twistMotion=Free
 }
 {bone1Name=dyn_shoeFlap_a_l, bone2Name=calf_twist_02_l, swing1LimitDegrees=40, swing2LimitDegrees=10, twistLimitDegrees=5}
 {bone1Name=dyn_shoeFlap_b_l, bone2Name=calf_twist_01_l, swing1LimitDegrees=40, swing2LimitDegrees=10, twistLimitDegrees=5}
 {bone1Name=dyn_shoeFlap_b_r, bone2Name=calf_twist_01_r, swing1LimitDegrees=40, swing2LimitDegrees=10, twistLimitDegrees=5}
 {bone1Name=dyn_shoeFlap_a_r, bone2Name=calf_twist_02_r, swing1LimitDegrees=40, swing2LimitDegrees=10, twistLimitDegrees=5}
]
