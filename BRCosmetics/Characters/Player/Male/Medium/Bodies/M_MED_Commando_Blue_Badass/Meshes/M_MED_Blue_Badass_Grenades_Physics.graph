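# Physics asset for the dangling grenade accessory on the Commando Blue Badass male medium body cosmetic.
# It has small capsule bodies on the two dynamic grenade bones and a larger capsule on the pelvis.
# Each grenade is constrained to the pelvis with narrow swing limits and a wider twist limit, which restricts its motion to a small range around the hip.
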PhysicsAsset /BRCosmetics/Characters/Player/Male/Medium/Bodies/M_MED_Commando_Blue_Badass/Meshes/M_MED_Blue_Badass_Grenades_Physics
bodies {
 dyn_grenade_a [{shapeType=Capsule, length=3, radius=3, shapeName=dyn_grenade_a_capsule, center=<0.7523 0 -5.8607>, rotation=<rot 0 0 10.0007>}]
 dyn_grenade_b [{shapeType=Capsule, length=3, radius=3, shapeName=dyn_grenade_b_capsule, center=<-0.6946 -0.0002 -6.1491>, rotation=<rot 0 0 10.0007>}]
 pelvis [{shapeType=Capsule, length=10, radius=10, shapeName=pelvis_capsule, center=<0 -0.9611 0.0001>}]
}
constraints [
 {bone1Name=dyn_grenade_a, bone2Name=pelvis, swing1LimitDegrees=10, swing2LimitDegrees=15, twistLimitDegrees=50}
 {bone1Name=dyn_grenade_b, bone2Name=pelvis, swing1LimitDegrees=10, swing2LimitDegrees=15, twistLimitDegrees=50}
]
