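# Post-copy AnimBP for the Blue Badass medium male Commando body, adding secondary motion to the collar and other dynamic bones.
# It copies the parent mesh pose, offsets collar base bones, then either runs three chained RigidBody sims with left-shoulder correctives, or applies head-up and shoulder bone correctives only.
# The update event derives corrective alphas from head and clavicle deltas, choosing the non-physics branch when rigid bodies are disabled or LOD exceeds 1.
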AnimBlueprint /BRCosmetics/Characters/Player/Male/Medium/Bodies/M_MED_Commando_Blue_Badass/Meshes/M_MED_Blue_Badass_AnimBP
target_skeleton=&/M_MED_Blue_Badass_Skeleton
graphs [
 AnimGraph {nodes [
   n1 = Get HeadUp() -> (HeadUp:double)
   n2 = "Transform (Modify) Bone - Bone: dyn_collar_back"(
    ComponentPose=Local To Component(LocalPose=Use cached pose 'SavedPose_common'())
    Alpha=@n1
    Translation="0,-1,0"
    Rotation="0,0,-25"
    Scale="1,1,1"
   ) {
    rotationMode=BMM_Additive
    rotationSpace=BCS_BoneSpace
    translationMode=BMM_Additive
    translationSpace=BCS_BoneSpace
    alphaBoolBlend {blendOption=Linear}
    boneToModify {boneName=dyn_collar_back}
    rotation {roll=-25}
    translation {y=-1}
   }
   n3 = "Transform (Modify) Bone - Bone: dyn_collar_side_l"(ComponentPose=@n2, Alpha=@n1, Translation="0,0,0", Rotation="0,0,-10", Scale="1,1,1") {
    rotationMode=BMM_Additive
    rotationSpace=BCS_BoneSpace
    translationMode=BMM_Additive
    translationSpace=BCS_BoneSpace
    alphaBoolBlend {blendOption=Linear}
    boneToModify {boneName=dyn_collar_side_l}
    rotation {roll=-10}
   }
   n4 = "Transform (Modify) Bone - Bone: dyn_collar_side_r"(ComponentPose=@n3, Alpha=@n1, Translation="0,0.5,0", Rotation="0,0,-20", Scale="1,1,1") {
    rotationMode=BMM_Additive
    rotationSpace=BCS_BoneSpace
    translationMode=BMM_Additive
    translationSpace=BCS_BoneSpace
    alphaBoolBlend {blendOption=Linear}
    boneToModify {boneName=dyn_collar_side_r}
    rotation {roll=-20}
    translation {y=0.5}
   }
   n5 = Get ShoulderUp_L() -> (ShoulderUp_L:double)
   n6 = "Transform (Modify) Bone - Bone: dyn_collar_front_l"(ComponentPose=@n4, Alpha=@n5, Translation="0,0,0", Rotation="0,0,25", Scale="1,1,1") {
    rotationMode=BMM_Additive
    rotationSpace=BCS_BoneSpace
    translationMode=BMM_Additive
    translationSpace=BCS_BoneSpace
    alphaBoolBlend {blendOption=Linear}
    boneToModify {boneName=dyn_collar_front_l}
    rotation {roll=25}
   }
   n7 = "Transform (Modify) Bone - Bone: dyn_collar_side_l"(ComponentPose=@n6, Alpha=@n5, Translation="0,0,0", Rotation="0,0,25", Scale="1,1,1") {
    rotationMode=BMM_Additive
    rotationSpace=BCS_BoneSpace
    translationMode=BMM_Additive
    translationSpace=BCS_BoneSpace
    alphaBoolBlend {blendOption=Linear}
    boneToModify {boneName=dyn_collar_side_l}
    rotation {roll=25}
   }
   n8 = RigidBody(ComponentPose=Local To Component(LocalPose=Use cached pose 'SavedPose_common'())) {
    bClampLinearTranslationLimitToRefPose=true
    bOverrideWorldGravity=true
    overridePhysicsAsset=&/M_MED_Blue_Badass_Dyn_Physics
    simulationSpace=BaseBoneSpace
    alphaBoolBlend {blendOption=Linear}
    baseBoneRef {boneName=root}
    componentAppliedLinearAccClamp=<3000 3000 3000>
    componentLinearAccScale=<0.3 0.3 0.7>
    componentLinearVelScale=<0.3 0.3 0.7>
    overrideWorldGravity {z=-980}
   }
   n9 = RigidBody(ComponentPose=@n8) {
    bClampLinearTranslationLimitToRefPose=true
    bOverrideWorldGravity=true
    overridePhysicsAsset=&/M_MED_Blue_Badass_Grenades_Physics
    simulationSpace=BaseBoneSpace
    alphaBoolBlend {blendOption=Linear}
    baseBoneRef {boneName=pelvis}
    componentAppliedLinearAccClamp=<3000 3000 3000>
    componentLinearAccScale=<0.3 0.3 0.7>
    componentLinearVelScale=<0.3 0.3 0.7>
    overrideWorldGravity {z=-980}
   }
   n10 = RigidBody(ComponentPose=@n9) {
    bClampLinearTranslationLimitToRefPose=true
    bOverrideWorldGravity=true
    overridePhysicsAsset=&/M_MED_Blue_Badass_Collar_Dyn_Physics
    simulationSpace=BaseBoneSpace
    alphaBoolBlend {blendOption=Linear}
    baseBoneRef {boneName=spine_05}
    overrideWorldGravity {z=-980}
   }
   n11 = Get ShoulderUp_L()
   n12 = "Transform (Modify) Bone - Bone: dyn_collar_front_l"(ComponentPose=@n10, Alpha=@n11, Translation="0,0,0", Rotation="0,0,25", Scale="1,1,1") {
    rotationMode=BMM_Additive
    rotationSpace=BCS_BoneSpace
    translationMode=BMM_Additive
    translationSpace=BCS_BoneSpace
    alphaBoolBlend {blendOption=Linear}
    boneToModify {boneName=dyn_collar_front_l}
    rotation {roll=25}
   }
   n13 = "Transform (Modify) Bone - Bone: dyn_collar_side_l"(ComponentPose=@n12, Alpha=@n11, Translation="0,0,0", Rotation="0,0,25", Scale="1,1,1") {
    rotationMode=BMM_Additive
    rotationSpace=BCS_BoneSpace
    translationMode=BMM_Additive
    translationSpace=BCS_BoneSpace
    alphaBoolBlend {blendOption=Linear}
    boneToModify {boneName=dyn_collar_side_l}
    rotation {roll=25}
   }
   Output Pose(Result=Blend Poses by bool(
     BlendPose_0=Component To Local(ComponentPose=@n7)
     BlendPose_1=Component To Local(ComponentPose=@n13)
     bActiveValue=Get isRigidBodyDisabled() -> (isRigidBodyDisabled:bool)
    ) {blendType=Linear}) {layerGroup=None}
   n14 = "Transform (Modify) Bone - Bone: dyn_collar_back_base"(
    ComponentPose=Local To Component(LocalPose=Copy Pose From Mesh() {bUseAttachedParent=true})
    Alpha=1
    Translation="0,-1,0"
    Rotation="0,0,0"
    Scale="1,1,1"
   ) {
    rotationMode=BMM_Additive
    rotationSpace=BCS_BoneSpace
    translationMode=BMM_Additive
    translationSpace=BCS_BoneSpace
    alphaBoolBlend {blendOption=Linear}
    boneToModify {boneName=dyn_collar_back_base}
    translation {y=-1}
   }
   n15 = "Transform (Modify) Bone - Bone: dyn_collar_side_r_base"(ComponentPose=@n14, Alpha=1, Translation="0,0.5,0", Rotation="0,0,0", Scale="1,1,1") {
    rotationMode=BMM_Additive
    rotationSpace=BCS_BoneSpace
    translationMode=BMM_Additive
    translationSpace=BCS_BoneSpace
    alphaBoolBlend {blendOption=Linear}
    boneToModify {boneName=dyn_collar_side_r_base}
    translation {y=0.5}
   }
   n16 = "Transform (Modify) Bone - Bone: dyn_collar_side_l_base"(ComponentPose=@n15, Alpha=1, Translation="0,-0.5,0", Rotation="0,0,0", Scale="1,1,1") {
    rotationMode=BMM_Additive
    rotationSpace=BCS_BoneSpace
    translationMode=BMM_Additive
    translationSpace=BCS_BoneSpace
    alphaBoolBlend {blendOption=Linear}
    boneToModify {boneName=dyn_collar_side_l_base}
    translation {y=-0.5}
   }
   n17 = "Transform (Modify) Bone - Bone: dyn_collar_front_r_base"(ComponentPose=@n16, Alpha=1, Translation="0,0.5,0", Rotation="0,0,0", Scale="1,1,1") {
    rotationMode=BMM_Additive
    rotationSpace=BCS_BoneSpace
    translationMode=BMM_Additive
    translationSpace=BCS_BoneSpace
    alphaBoolBlend {blendOption=Linear}
    boneToModify {boneName=dyn_collar_front_r_base}
    translation {y=0.5}
   }
   Save cached pose 'SavedPose_common'(Pose=Component To Local(ComponentPose=@n17)) {cacheName=SavedPose_common}
  ]}
 EventGraph {nodes [
   Event BlueprintUpdateAnimation()
   > n1 = Sequence!(execute) -> (then_0, then_1, then_2:Exec, then_3:Exec, then_4:Exec, then_5:Exec, then_6:Exec, then_7:Exec)
   n2 = GetOwnerSkeletalMeshComponent()
   n3 = BreakRotator(InRot=BreakTransform(InTransform=GetDeltaTransformFromRefPose(self=@n2, BoneName=head, BaseName=spine_05)).Rotation)
   Set HeadUp(execute=@n1.then_1, HeadUp:double=MapRangeClamped(Value=@n3.Yaw, InRangeA=0, InRangeB=-44)) -> (Output_Get:double)
   n4 = BreakTransform(InTransform=GetDeltaTransformFromRefPose(self=@n2, BoneName=clavicle_l, BaseName=spine_05))
   Set ShoulderUp_L(execute=@n1.then_4, ShoulderUp_L:double=MapRangeClamped(Value=BreakRotator(InRot=@n4.Rotation).Pitch, InRangeA=30, InRangeB=45)) -> (Output_Get:double)
   n5 = Get CurrentLOD()
   n6 = Get bRigidBodyEnabled()
   Set isRigidBodyDisabled(
    execute=@n1.then_6
    isRigidBodyDisabled:bool=OR Boolean(A=NOT Boolean(A=@n6), B="integer > integer"(A=@n5, B=1) {is=title K2Node_CallFunction})
   ) -> (Output_Get:bool)
   TryGetPawnOwner()
   n7 = BreakRotator(InRot=BreakTransform(InTransform=GetDeltaTransformFromRefPose(self=@n2, BoneName=head, BaseName=neck_01)).Rotation)
   Set HeadTwist_R(HeadTwist_R:double=MapRangeClamped(Value=@n7, InRangeA=0, InRangeB=-25)) -> (Output_Get:double)
   Set HeadTwist_L(HeadTwist_L:double=MapRangeClamped(Value=@n7, InRangeA=0, InRangeB=25)) -> (Output_Get:double)
   Set HeadDn(HeadDn:double=MapRangeClamped(Value=@n3.Yaw, InRangeA=17, InRangeB=0, OutRangeA=1, OutRangeB=0)) -> (Output_Get:double)
  ]}
]
variables [
 HeadDn {type=number}
 HeadTwist_L {type=number}
 HeadTwist_R {type=number}
 HeadUp {type=number}
 isRigidBodyDisabled {type=boolean}
 ShoulderUp_L {type=number}
]
