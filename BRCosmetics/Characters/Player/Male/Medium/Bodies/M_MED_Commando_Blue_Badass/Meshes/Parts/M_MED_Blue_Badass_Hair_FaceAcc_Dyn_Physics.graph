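# Physics asset for the dynamic hair and face accessory mesh part of the medium male Blue Badass Commando player character.
# Twelve small dyn_hair capsules, mostly zero-length and therefore spherical, surround one larger head capsule.
# Each hair body is constrained directly to the head rather than chained, with a narrow second swing limit throughout.
# Hair bones 0, 9 and 10 have tighter swing and twist limits than the rest.
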PhysicsAsset /BRCosmetics/Characters/Player/Male/Medium/Bodies/M_MED_Commando_Blue_Badass/Meshes/Parts/M_MED_Blue_Badass_Hair_FaceAcc_Dyn_Physics
bodies {
 dyn_hair_0 [{shapeType=Capsule, length=0.4418, radius=0.5, center=<0 0 -0.5>, rotation=<rot 0 0 89.9998>}]
 dyn_hair_1 [{shapeType=Capsule, length=0, radius=0.5391, center=<0 0 -0.5>}]
 dyn_hair_10 [{shapeType=Capsule, length=0, radius=0.7385, center=<0 0 -0.5>}]
 dyn_hair_11 [{shapeType=Capsule, length=0, radius=0.5582, center=<0 0 -0.5>}]
 dyn_hair_2 [{shapeType=Capsule, length=0, radius=0.5, center=<0 0 -0.5>}]
 dyn_hair_3 [{shapeType=Capsule, length=0, radius=0.6232, center=<0 0 -0.5>}]
 dyn_hair_4 [{shapeType=Capsule, length=0, radius=0.5, center=<0 0 -0.5>}]
 dyn_hair_5 [{shapeType=Capsule, length=0.25, radius=0.5602, center=<0 0 -0.3319>}]
 dyn_hair_6 [{shapeType=Capsule, length=0, radius=0.6035, center=<0 0 -0.5>}]
 dyn_hair_7 [{shapeType=Capsule, length=0, radius=0.5158, center=<0 0 -0.5>}]
 dyn_hair_8 [{shapeType=Capsule, length=0, radius=0.5, center=<0 0 -0.5>}]
 dyn_hair_9 [{shapeType=Capsule, length=0, radius=0.505, center=<0 0 -0.5>}]
 head [{shapeType=Capsule, length=3.9438, radius=5.6392, rotation=<rot -90 0.0099 -0.0099>}]
}
constraints [
 {bone1Name=dyn_hair_0, bone2Name=head, swing1LimitDegrees=10, swing2LimitDegrees=5, twistLimitDegrees=10}
 {bone1Name=dyn_hair_1, bone2Name=head, swing1LimitDegrees=30, swing2LimitDegrees=5, twistLimitDegrees=30}
 {bone1Name=dyn_hair_2, bone2Name=head, swing1LimitDegrees=30, swing2LimitDegrees=5, twistLimitDegrees=45}
 {bone1Name=dyn_hair_3, bone2Name=head, swing1LimitDegrees=30, swing2LimitDegrees=5, twistLimitDegrees=30}
 {bone1Name=dyn_hair_4, bone2Name=head, swing1LimitDegrees=30, swing2LimitDegrees=5, twistLimitDegrees=45}
 {bone1Name=dyn_hair_5, bone2Name=head, swing1LimitDegrees=30, swing2LimitDegrees=5, twistLimitDegrees=30}
 {bone1Name=dyn_hair_6, bone2Name=head, swing1LimitDegrees=30, swing2LimitDegrees=5, twistLimitDegrees=45}
 {bone1Name=dyn_hair_7, bone2Name=head, swing1LimitDegrees=30, swing2LimitDegrees=5, twistLimitDegrees=45}
 {bone1Name=dyn_hair_8, bone2Name=head, swing1LimitDegrees=30, swing2LimitDegrees=5, twistLimitDegrees=45}
 {bone1Name=dyn_hair_9, bone2Name=head, swing1LimitDegrees=10, swing2LimitDegrees=5, twistLimitDegrees=10}
 {bone1Name=dyn_hair_10, bone2Name=head, swing1LimitDegrees=10, swing2LimitDegrees=5, twistLimitDegrees=10}
 {bone1Name=dyn_hair_11, bone2Name=head, swing1LimitDegrees=30, swing2LimitDegrees=5, twistLimitDegrees=30}
]
