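# Anim Blueprint that adds secondary physics motion to the hair and face accessory mesh of the male Commando Blue Badass cosmetic body.
# It copies the pose from the attached parent mesh, then runs twelve chained AnimDynamics nodes, one per dyn_hair bone.
# Each uses root-relative simulation with angular springs, damping and reduced gravity, differing mainly in angular limits and alpha.
# A RigidBody node sits unconnected, and the event graph holds only unwired stubs.
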AnimBlueprint /BRCosmetics/Characters/Player/Male/Medium/Bodies/M_MED_Commando_Blue_Badass/Meshes/Parts/M_MED_Blue_Badass_Hair_FaceAcc_AnimBP
target_skeleton=&/M_MED_Blue_Badass_Hair_FaceAcc_Skeleton
graphs [
 AnimGraph {nodes [
   RigidBody() {
    bClampLinearTranslationLimitToRefPose=true
    bOverrideWorldGravity=true
    overridePhysicsAsset=&/M_MED_Blue_Badass_Hair_FaceAcc_Dyn_Physics
    simulationSpace=BaseBoneSpace
    alphaBoolBlend {blendOption=Linear}
    baseBoneRef {boneName=spine_03}
    componentLinearAccScale=<0.3 0.3 0.3>
    componentLinearVelScale=<0.3 0.3 0.3>
    overrideWorldGravity {z=-980}
   }
   n1 = Anim Dynamics - Bone: dyn_hair_0(ComponentPose=Local To Component(LocalPose=Copy Pose From Mesh() {bUseAttachedParent=true}), Alpha=1, ExternalForce="0,0,0") {
    angularDampingOverride=1
    angularSpringConstant=500
    bAngularSpring=true
    bGravityOverrideInSimSpace=true
    bOverrideAngularDamping=true
    gravityScale=0.35
    simulationSpace=RootRelative
    alphaBoolBlend {blendOption=Linear}
    boundBone {boneName=dyn_hair_0}
    componentAppliedLinearAccClamp=<3000 3000 3000>
    componentLinearAccScale=<0.3 0.3 0.3>
    componentLinearVelScale=<0.3 0.3 0.3>
    physicsBodyDefinitions [{
      sphereCollisionRadius=0
      boundBone {boneName=dyn_hair_0}
      boxExtents=<5 5 5>
      constraintSetup {twistAxis=AxisY, angularLimitsMax=<15 0 15>, angularLimitsMin=<-15 0 -15>, angularTarget=<0 0 0>}
      localJointOffset=<0 0 -1>
     }]
   }
   n2 = Anim Dynamics - Bone: dyn_hair_1(ComponentPose=@n1, Alpha=0.85, ExternalForce="0,0,0") {
    angularDampingOverride=1
    angularSpringConstant=500
    bAngularSpring=true
    bGravityOverrideInSimSpace=true
    bOverrideAngularDamping=true
    gravityScale=0.35
    simulationSpace=RootRelative
    alphaBoolBlend {blendOption=Linear}
    boundBone {boneName=dyn_hair_1}
    componentAppliedLinearAccClamp=<3000 3000 3000>
    componentLinearAccScale=<0.3 0.3 0.3>
    componentLinearVelScale=<0.3 0.3 0.3>
    physicsBodyDefinitions [{
      sphereCollisionRadius=0
      boundBone {boneName=dyn_hair_1}
      boxExtents=<5 5 5>
      constraintSetup {twistAxis=AxisY, angularLimitsMax=<20 0 20>, angularLimitsMin=<-20 0 -20>, angularTarget=<0 0 0>}
      localJointOffset=<0 0 -1>
     }]
   }
   n3 = Anim Dynamics - Bone: dyn_hair_2(ComponentPose=@n2, Alpha=1, ExternalForce="0,0,0") {
    angularDampingOverride=1
    angularSpringConstant=500
    bAngularSpring=true
    bGravityOverrideInSimSpace=true
    bOverrideAngularDamping=true
    gravityScale=0.35
    simulationSpace=RootRelative
    alphaBoolBlend {blendOption=Linear}
    boundBone {boneName=dyn_hair_2}
    componentAppliedLinearAccClamp=<3000 3000 3000>
    componentLinearAccScale=<0.3 0.3 0.3>
    componentLinearVelScale=<0.3 0.3 0.3>
    physicsBodyDefinitions [{
      sphereCollisionRadius=0
      boundBone {boneName=dyn_hair_2}
      boxExtents=<5 5 5>
      constraintSetup {twistAxis=AxisY, angularLimitsMax=<5 0 30>, angularLimitsMin=<-5 0 -30>, angularTarget=<0 0 0>}
      localJointOffset=<0 0 -1>
     }]
   }
   n4 = Anim Dynamics - Bone: dyn_hair_3(ComponentPose=@n3, Alpha=1, ExternalForce="0,0,0") {
    angularDampingOverride=1
    angularSpringConstant=500
    bAngularSpring=true
    bGravityOverrideInSimSpace=true
    bOverrideAngularDamping=true
    gravityScale=0.35
    simulationSpace=RootRelative
    alphaBoolBlend {blendOption=Linear}
    boundBone {boneName=dyn_hair_3}
    componentAppliedLinearAccClamp=<3000 3000 3000>
    componentLinearAccScale=<0.3 0.3 0.3>
    componentLinearVelScale=<0.3 0.3 0.3>
    physicsBodyDefinitions [{
      sphereCollisionRadius=0
      boundBone {boneName=dyn_hair_3}
      boxExtents=<5 5 5>
      constraintSetup {twistAxis=AxisY, angularLimitsMax=<30 0 30>, angularLimitsMin=<-30 0 -30>, angularTarget=<0 0 0>}
      localJointOffset=<0 0 -1>
     }]
   }
   n5 = Anim Dynamics - Bone: dyn_hair_4(ComponentPose=@n4, Alpha=1, ExternalForce="0,0,0") {
    angularDampingOverride=1
    angularSpringConstant=500
    bAngularSpring=true
    bGravityOverrideInSimSpace=true
    bOverrideAngularDamping=true
    gravityScale=0.35
    simulationSpace=RootRelative
    alphaBoolBlend {blendOption=Linear}
    boundBone {boneName=dyn_hair_4}
    componentAppliedLinearAccClamp=<3000 3000 3000>
    componentLinearAccScale=<0.3 0.3 0.3>
    componentLinearVelScale=<0.3 0.3 0.3>
    physicsBodyDefinitions [{
      sphereCollisionRadius=0
      boundBone {boneName=dyn_hair_4}
      boxExtents=<5 5 5>
      constraintSetup {twistAxis=AxisY, angularLimitsMax=<30 0 30>, angularLimitsMin=<-30 0 -30>, angularTarget=<0 0 0>}
      localJointOffset=<0 0 -1>
     }]
   }
   n6 = Anim Dynamics - Bone: dyn_hair_5(ComponentPose=@n5, Alpha=0.7, ExternalForce="0,0,0") {
    angularDampingOverride=1
    angularSpringConstant=500
    bAngularSpring=true
    bGravityOverrideInSimSpace=true
    bOverrideAngularDamping=true
    gravityScale=0.35
    simulationSpace=RootRelative
    alphaBoolBlend {blendOption=Linear}
    boundBone {boneName=dyn_hair_5}
    componentAppliedLinearAccClamp=<3000 3000 3000>
    componentLinearAccScale=<0.3 0.3 0.3>
    componentLinearVelScale=<0.3 0.3 0.3>
    physicsBodyDefinitions [{
      sphereCollisionRadius=0
      boundBone {boneName=dyn_hair_5}
      boxExtents=<5 5 5>
      constraintSetup {twistAxis=AxisY, angularLimitsMax=<20 0 20>, angularLimitsMin=<-20 0 -20>, angularTarget=<0 0 0>}
      localJointOffset=<0 0 -1>
     }]
   }
   n7 = Anim Dynamics - Bone: dyn_hair_6(ComponentPose=@n6, Alpha=1, ExternalForce="0,0,0") {
    angularDampingOverride=1
    angularSpringConstant=500
    bAngularSpring=true
    bGravityOverrideInSimSpace=true
    bOverrideAngularDamping=true
    gravityScale=0.35
    simulationSpace=RootRelative
    alphaBoolBlend {blendOption=Linear}
    boundBone {boneName=dyn_hair_6}
    componentAppliedLinearAccClamp=<3000 3000 3000>
    componentLinearAccScale=<0.3 0.3 0.3>
    componentLinearVelScale=<0.3 0.3 0.3>
    physicsBodyDefinitions [{
      sphereCollisionRadius=0
      boundBone {boneName=dyn_hair_6}
      boxExtents=<5 5 5>
      constraintSetup {twistAxis=AxisY, angularLimitsMax=<30 0 30>, angularLimitsMin=<-30 0 -30>, angularTarget=<0 0 0>}
      localJointOffset=<0 0 -1>
     }]
   }
   n8 = Anim Dynamics - Bone: dyn_hair_7(ComponentPose=@n7, Alpha=1, ExternalForce="0,0,0") {
    angularDampingOverride=1
    angularSpringConstant=500
    bAngularSpring=true
    bGravityOverrideInSimSpace=true
    bOverrideAngularDamping=true
    gravityScale=0.35
    simulationSpace=RootRelative
    alphaBoolBlend {blendOption=Linear}
    boundBone {boneName=dyn_hair_7}
    componentAppliedLinearAccClamp=<3000 3000 3000>
    componentLinearAccScale=<0.3 0.3 0.3>
    componentLinearVelScale=<0.3 0.3 0.3>
    physicsBodyDefinitions [{
      sphereCollisionRadius=0
      boundBone {boneName=dyn_hair_7}
      boxExtents=<5 5 5>
      constraintSetup {twistAxis=AxisY, angularLimitsMax=<30 0 30>, angularLimitsMin=<-30 0 -30>, angularTarget=<0 0 0>}
      localJointOffset=<0 0 -1>
     }]
   }
   n9 = Anim Dynamics - Bone: dyn_hair_8(ComponentPose=@n8, Alpha=1, ExternalForce="0,0,0") {
    angularDampingOverride=1
    angularSpringConstant=500
    bAngularSpring=true
    bGravityOverrideInSimSpace=true
    bOverrideAngularDamping=true
    gravityScale=0.35
    simulationSpace=RootRelative
    alphaBoolBlend {blendOption=Linear}
    boundBone {boneName=dyn_hair_8}
    componentAppliedLinearAccClamp=<3000 3000 3000>
    componentLinearAccScale=<0.3 0.3 0.3>
    componentLinearVelScale=<0.3 0.3 0.3>
    physicsBodyDefinitions [{
      sphereCollisionRadius=0
      boundBone {boneName=dyn_hair_8}
      boxExtents=<5 5 5>
      constraintSetup {twistAxis=AxisY, angularLimitsMax=<30 0 30>, angularLimitsMin=<-30 0 -30>, angularTarget=<0 0 0>}
      localJointOffset=<0 0 -1>
     }]
   }
   n10 = Anim Dynamics - Bone: dyn_hair_9(ComponentPose=@n9, Alpha=1, ExternalForce="0,0,0") {
    angularDampingOverride=1
    angularSpringConstant=500
    bAngularSpring=true
    bGravityOverrideInSimSpace=true
    bOverrideAngularDamping=true
    gravityScale=0.35
    simulationSpace=RootRelative
    alphaBoolBlend {blendOption=Linear}
    boundBone {boneName=dyn_hair_9}
    componentAppliedLinearAccClamp=<3000 3000 3000>
    componentLinearAccScale=<0.3 0.3 0.3>
    componentLinearVelScale=<0.3 0.3 0.3>
    physicsBodyDefinitions [{
      sphereCollisionRadius=0
      boundBone {boneName=dyn_hair_9}
      boxExtents=<5 5 5>
      constraintSetup {twistAxis=AxisZ, angularLimitsMax=<30 30 0>, angularLimitsMin=<-30 -30 0>, angularTarget=<0 0 0>}
      localJointOffset=<0 0 -1>
     }]
   }
   n11 = Anim Dynamics - Bone: dyn_hair_10(ComponentPose=@n10, Alpha=1, ExternalForce="0,0,0") {
    angularDampingOverride=1
    angularSpringConstant=500
    bAngularSpring=true
    bGravityOverrideInSimSpace=true
    bOverrideAngularDamping=true
    gravityScale=0.35
    simulationSpace=RootRelative
    alphaBoolBlend {blendOption=Linear}
    boundBone {boneName=dyn_hair_10}
    componentAppliedLinearAccClamp=<3000 3000 3000>
    componentLinearAccScale=<0.3 0.3 0.3>
    componentLinearVelScale=<0.3 0.3 0.3>
    physicsBodyDefinitions [{
      sphereCollisionRadius=0
      boundBone {boneName=dyn_hair_10}
      boxExtents=<5 5 5>
      constraintSetup {twistAxis=AxisZ, angularLimitsMax=<0 30 0>, angularLimitsMin=<-30 -30 0>, angularTarget=<0 0 0>}
      localJointOffset=<0 0 -1>
     }]
   }
   n12 = Anim Dynamics - Bone: dyn_hair_11(ComponentPose=@n11, Alpha=1, ExternalForce="0,0,0") {
    angularDampingOverride=1
    angularSpringConstant=500
    bAngularSpring=true
    bGravityOverrideInSimSpace=true
    bOverrideAngularDamping=true
    gravityScale=0.35
    simulationSpace=RootRelative
    alphaBoolBlend {blendOption=Linear}
    boundBone {boneName=dyn_hair_11}
    componentAppliedLinearAccClamp=<3000 3000 3000>
    componentLinearAccScale=<0.3 0.3 0.3>
    componentLinearVelScale=<0.3 0.3 0.3>
    physicsBodyDefinitions [{
      sphereCollisionRadius=0
      boundBone {boneName=dyn_hair_11}
      boxExtents=<5 5 5>
      constraintSetup {twistAxis=AxisY, angularLimitsMax=<50 0 50>, angularLimitsMin=<-50 0 -50>, angularTarget=<0 0 0>}
      localJointOffset=<0 0 -1>
     }]
   }
   Output Pose(Result=Component To Local(ComponentPose=@n12)) {layerGroup=None}
  ]}
 EventGraph {nodes [
   TryGetPawnOwner()
   Event BlueprintUpdateAnimation()
  ]}
]
variables []
